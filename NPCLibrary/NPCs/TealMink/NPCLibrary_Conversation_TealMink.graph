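# Conversation database for the TealMink NPC, defining its greeting and the menu of branches available when a player talks to it.
# One entry point feeds a Speech node whose default gearhead greeting can be replaced by prioritized contextual lines for first-ever conversations, match conversation counts, or specific player outfits.
# That node fans out to link nodes into other entry tags: quests, services, TealMink dialogue, assault rifle expertise and a Season 34 story quest.
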ConversationDatabase /NPCLibrary/NPCs/TealMink/NPCLibrary_Conversation_TealMink
entry_points [{tag=Conversation.Entry.TealMink, destinations [#1]}]
nodes [
 ConversationEntryPointNode_7 {
  id=#1
  class=ConversationEntryPointNode
  outputs [#2]
  properties {entryTag {tagName=Conversation.Entry.TealMink}}
 }
 Speech_C_4 {
  id=#2
  class=Speech_C
  outputs [
   #3
   #4
   #5
   #6
   #7
   #8
   #9
   #10
  ]
  properties {
   nodeName=""
   generalConfig {
    defaultMessage="Howdy, gearhead."
    contextualMessages [
     {
      message=Us gearheads have one rule: don't cause trouble in the shop.
      contextRequirements [{requirement {class=FortControllerRequirement_IsFirstEverConversationWithNPC}}]
      priority {curve {rowName=Tandem.Speech.Priority.FirstEverInteraction}}
     }
     {
      message="Howdy, gearhead."
      contextRequirements [{requirement {class=FortControllerRequirement_CheckNPCMatchConversationCount, properties {nPCUniqueIdTag {tagName=AISpawnerData.Type.Tandem.TealMink}}}}]
      priority {curve {rowName=None}}
     }
     {
      message="Tech and gear, available here! Ugh, that rhymed."
      contextRequirements [{requirement {class=FortControllerRequirement_CheckNPCMatchConversationCount, properties {nPCUniqueIdTag {tagName=AISpawnerData.Type.Tandem.TealMink}}}}]
      priority {curve {rowName=None}}
     }
     {
      message=Welcome back.
      contextRequirements [{requirement {class=FortControllerRequirement_CheckNPCMatchConversationCount, properties {nPCUniqueIdTag {tagName=AISpawnerData.Type.Tandem.TealMink}}}}]
      priority {curve {rowName=None}}
     }
     {
      message="You promised you'd stay safe out there, Cass."
      contextRequirements [{requirement {class=FortControllerRequirement_HasCID, properties {allowedCIds [/BRCosmetics/Athena/Items/Cosmetics/Characters/Character_CasinoReaper_Die]}}}]
      priority {curve {rowName=Tandem.Speech.Priority.OutfitReaction}}
     }
     {
      message=I don't have the parts to service your car. Sorry.
      contextRequirements [{requirement {class=FortControllerRequirement_HasCID, properties {allowedCIds [/BRCosmetics/Athena/Items/Cosmetics/Characters/Character_SherWolf_Gnash]}}}]
      priority {curve {rowName=Tandem.Speech.Priority.OutfitReaction}}
     }
     {
      message=You've got an angle... don't you?
      contextRequirements [{requirement {class=FortControllerRequirement_HasCID, properties {allowedCIds [&Character_AuraCop_Bullet]}}}]
      priority {curve {rowName=Tandem.Speech.Priority.OutfitReaction}}
     }
    ]
   }
  }
 }
 ConversationLinkNode_7 {id=#10, class=ConversationLinkNode, properties {remoteEntryTag {tagName=Conversation.Entry.DeluxeServices}}}
 ConversationLinkNode_6 {
  id=#9
  class=ConversationLinkNode
  properties {remoteEntryTag {tagName=Conversation.Entry.Season34.Story.P04.Q02}}
 }
 ConversationLinkNode_5 {id=#8, class=ConversationLinkNode, properties {remoteEntryTag {tagName=Conversation.Entry.Expertise.Assault}}}
 ConversationLinkNode_4 {
  id=#7
  class=ConversationLinkNode
  properties {remoteEntryTag {tagName=Conversation.Entry.Season34.Expertise.AssaultRifle.Finish}}
 }
 ConversationLinkNode_3 {
  id=#6
  class=ConversationLinkNode
  properties {remoteEntryTag {tagName=Conversation.Entry.Season34.Expertise.AssaultRifle.Introduction}}
 }
 ConversationLinkNode_2 {id=#5, class=ConversationLinkNode, properties {remoteEntryTag {tagName=Conversation.Entry.TealMink.Dialogue}}}
 ConversationLinkNode_0 {id=#4, class=ConversationLinkNode, properties {remoteEntryTag {tagName=Conversation.Entry.BasicServices}}}
 ConversationLinkNode_1 {id=#3, class=ConversationLinkNode, properties {remoteEntryTag {tagName=Conversation.Entry.NPCLibrary.Quests}}}
]
