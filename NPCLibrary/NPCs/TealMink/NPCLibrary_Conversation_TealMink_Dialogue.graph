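# Conversation database for the TealMink NPC's "Ask about..." dialogue, where players hear character and world lore.
# A single entry tag leads to an opening Speech node that branches into Cassidy Quinn, The Black Markets and Motorcycles topics, a BACK node, and a link to the separate BaronEsc dialogue entry.
# Each topic plays a linear chain of Speech lines, and all three converge on one shared ReturnToStart node.
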ConversationDatabase /NPCLibrary/NPCs/TealMink/NPCLibrary_Conversation_TealMink_Dialogue
entry_points [{tag=Conversation.Entry.TealMink.Dialogue, destinations [#1]}]
nodes [
 ConversationEntryPointNode_0 {
  id=#1
  class=ConversationEntryPointNode
  outputs [#2]
  properties {entryTag {tagName=Conversation.Entry.TealMink.Dialogue}}
 }
 Speech_C_0 {
  id=#2
  class=Speech_C
  outputs [
   #3
   #4
   #5
   #6
   #7
  ]
  properties {
   nodeName=""
   softTaskIcon=&T_UI_NPC_question
   generalConfig {defaultMessage=I know just about everybody around here.}
  }
  sub_nodes [ConversationChoiceNode_5 {class=ConversationChoiceNode, properties {defaultChoiceDisplayText=Ask about...}}]
 }
 ConversationLinkNode_0 {
  id=#7
  class=ConversationLinkNode
  properties {remoteEntryTag {tagName=Conversation.Entry.TealMink.Dialogue.BaronEsc}}
 }
 Speech_C_6 {
  id=#6
  class=Speech_C
  outputs [#8]
  properties {
   nodeName=""
   softTaskIcon=&T_UI_NPC_question
   generalConfig {defaultMessage="You love bikes too?! I swear, I was put on this Island to build 'em and ride 'em."}
  }
  sub_nodes [ConversationChoiceNode_15 {class=ConversationChoiceNode, properties {defaultChoiceDisplayText=Motorcycles}}]
 }
 Speech_C_8 {
  id=#8
  class=Speech_C
  outputs [#9]
  properties {
   nodeName=""
   softTaskIcon=&T_UI_NPC_question
   generalConfig {defaultMessage="Cassidy and I used to... \"borrow\" bikes together all the time."}
  }
 }
 Speech_C_9 {
  id=#9
  class=Speech_C
  outputs [#10]
  properties {nodeName="", softTaskIcon=&T_UI_NPC_question, generalConfig {defaultMessage="I've gone legit... mostly. "}}
 }
 Speech_C_10 {
  id=#10
  class=Speech_C
  outputs [#11]
  properties {
   nodeName=""
   softTaskIcon=&T_UI_NPC_question
   generalConfig {defaultMessage="But if Cass and I hadn't boosted bikes as kids, who knows what I'd be doing now!"}
  }
 }
 ReturnToStart_C_23 {
  id=#11
  class=ReturnToStart_C
  properties {softTaskIcon=&T_UI_NPC_Back}
  sub_nodes [ConversationChoiceNode_32 {class=ConversationChoiceNode, properties {defaultChoiceDisplayText=BACK}}]
 }
 Speech_C_2 {
  id=#5
  class=Speech_C
  outputs [#12]
  properties {
   nodeName=""
   softTaskIcon=&T_UI_NPC_question
   generalConfig {defaultMessage="In a place like this, with Kane's thumb pressin' down on all of us..."}
  }
  sub_nodes [ConversationChoiceNode_1 {class=ConversationChoiceNode, properties {defaultChoiceDisplayText=The Black Markets}}]
 }
 Speech_C_3 {
  id=#12
  class=Speech_C
  outputs [#13]
  properties {
   nodeName=""
   softTaskIcon=&T_UI_NPC_question
   generalConfig {defaultMessage=...I think of our markets as a community good.}
  }
 }
 Speech_C_4 {
  id=#13
  class=Speech_C
  outputs [#14]
  properties {
   nodeName=""
   softTaskIcon=&T_UI_NPC_question
   generalConfig {defaultMessage="Even outlaws need a place to congregate, kick back, and network, you know?"}
  }
 }
 Speech_C_5 {
  id=#14
  class=Speech_C
  outputs [#11]
  properties {
   nodeName=""
   softTaskIcon=&T_UI_NPC_question
   generalConfig {defaultMessage="And the more Bars I make, the more bikes I can build. It's a win-win."}
  }
 }
 Back_C_0 {
  id=#4
  class=Back_C
  sub_nodes [ConversationChoiceNode_7 {class=ConversationChoiceNode, properties {defaultChoiceDisplayText=BACK}}]
 }
 Speech_C_7 {
  id=#3
  class=Speech_C
  outputs [#15]
  properties {
   nodeName=""
   softTaskIcon=&T_UI_NPC_question
   generalConfig {defaultMessage="Cassidy's my girl! We grew up together."}
  }
  sub_nodes [ConversationChoiceNode_0 {class=ConversationChoiceNode, properties {defaultChoiceDisplayText=Cassidy Quinn}}]
 }
 Speech_C_15 {
  id=#15
  class=Speech_C
  outputs [#16]
  properties {
   nodeName=""
   softTaskIcon=&T_UI_NPC_question
   generalConfig {defaultMessage=She's a thrillseeker. I'm a worrier. What else can I say?}
  }
 }
 Speech_C_14 {
  id=#16
  class=Speech_C
  outputs [#17]
  properties {
   nodeName=""
   softTaskIcon=&T_UI_NPC_question
   generalConfig {defaultMessage="Oh! I guess I could say: if you mess with her..."}
  }
 }
 Speech_C_1 {
  id=#17
  class=Speech_C
  outputs [#11]
  properties {
   nodeName=""
   softTaskIcon=&T_UI_NPC_question
   generalConfig {defaultMessage="...we're going to have a serious, serious problem. Got it?"}
  }
 }
]
